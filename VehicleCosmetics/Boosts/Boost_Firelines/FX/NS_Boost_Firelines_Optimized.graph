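# Niagara system for the Firelines vehicle boost cosmetic, trailing glowing ribbons and flying sprites from skeletal mesh sockets, one set per port.
# Particles_Flying partitions sprites across sockets, throwing them backward with vortex swirl, shrink and fade.
# Ribbons spawns trails with age-driven sine wobble in velocity, drawn by three renderers in one emitter through extra color and width bindings.
# Both emitters inherit source movement and take painted colors and car offset from user parameters.
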NiagaraSystem /VehicleCosmetics/Boosts/Boost_Firelines/FX/NS_Boost_Firelines_Optimized
dependencies {
 usedDataInterfaces [&NiagaraDataInterfaceCurve, &NiagaraDataInterfaceVectorCurve, &NiagaraDataInterfaceSkeletalMesh]
 usedDynamicInputs []
 usedModules [
  &self:SystemSpawnScript.SystemScriptSource.SystemScriptGraph.NiagaraNodeAssignment_0.SetVariables_#1
  &SystemState
 ]
}
emitters {
 Particles_Flying {
  emitterUpdateScript [
   &EmitterState()
   &VC_FilteredSocketCount()
   &SpawnRate(SpawnRate=&RandomRangeFloat(Minimum=20, Maximum=40), Use Spawn Probability=false)
  ]
  particleSpawnScript [
   &VC_PaintedVariantsOverride() {enabled=false}
   &InitializeParticle(
    Color=$User.PaintedColor_4
    Lifetime Max=0.7
    Lifetime Min=0.6
    Lifetime Mode=::NewEnumerator1
    Mass Mode=::NewEnumerator3
    Sprite Rotation Mode=::NewEnumerator2
    Sprite Size Mode=::NewEnumerator1
    Uniform Sprite Size Max=15
    Uniform Sprite Size Min=9
    UsePositionOffset=true
    Write Material Random=true
   )
   &PartitionParticles(Number Of Partitions=$Emitter.PortCount)
   &SkeletalMeshLocation(
    Bone / Socket Index=$Particles.PartitionParticles.Partition
    Bone / Socket Sampling Mode=::NewEnumerator5
    MaskSpawnGroup=false
    Mesh Sampling Type=::NewEnumerator6
    Sampled Position Offset=&AddVectorToPosition(Position=<0 0 0>, Vector=$User.DTCarOffset)
    Skeletal Mesh=$System.SkeletalMesh
   )
   &AddVelocity(Velocity=<-800 0 0>)
   &ShapeLocation(
    MaskSpawnGroup=false
    Override Seed=false
    Sphere Radius=&RandomRangeFloat(Minimum=30, Maximum=40)
    UseOwnerScale=true
   )
  ]
  particleUpdateScript [
   &ParticleState()
   &ScaleSpriteSize(Uniform Curve Sprite Scale=DataInterface() {propertyValues {Curve {keys [{tangentMode=RCTM_User, time=0, value=1}, {tangentMode=RCTM_User, time=1.8, value=0}]}}})
   &VortexForce(
    UseOriginOffset=false
    Vortex Axis=&RandomRangeVector(Minimum=<0 0 0>, Maximum=<0 0 0>)
    Vortex Force Amount=100
   )
   &InheritSourceMovement(
    Applied Position Scale=&Multiply_VectorByFloat(
     Vector=<1 1 1>
     Float=&FloatFromCurve(FloatCurve=DataInterface() {propertyValues {Curve {keys [
          {interpMode=RCIM_Cubic, leaveTangent=-0.0051, tangentMode=RCTM_Break, time=0, value=1}
          {arriveTangent=-0.247, interpMode=RCIM_Cubic, tangentMode=RCTM_Break, time=1, value=0.9}
         ]}}})
    )
    ApplyForce=false
    ApplyPosition=true
    ApplyPositionFalloff=false
   )
   &SolveForcesAndVelocity(Clamp Velocity=false, Limit Acceleration=false)
   &ScaleColor(
    Scale Alpha=&FloatFromCurve(FloatCurve=DataInterface() {propertyValues {Curve {keys [<at -0.0046 1.0113 RCIM_Cubic>,<at 1.0011 0.002 RCIM_Cubic>]}}}, Scale Curve=3)
    Scale RGB=<10 10 10>
    ScaleA=true
    ScaleRGB=true
   )
   &DynamicMaterialParameters(
    Index 0 Param 1=&RandomRangeFloat(Minimum=100, Maximum=300)
    Index 0 Param 2=&RandomRangeFloat(Minimum=-0.45, Maximum=-0.7)
    Index 0 Param 4=&RandomRangeFloat(Minimum=0, Maximum=2)
    Module.Index0Param0Write=true
    Module.Index0Param1Write=true
    Module.Index0Param2Write=true
    Module.Index0Param3Write=true
   )
  ]
  rendererClasses [&NiagaraSpriteRendererProperties]
  renderers [{rendererClass=&NiagaraSpriteRendererProperties}]
 }
 Ribbons {
  emitterUpdateScript [
   &VC_RateQualityHandler(Epic Spawn Rate=18, High Spawn Rate=18, Low Spawn Rate=18, Medium Spawn Rate=18, Use Custom Rate?=true) {module=VC_RateQualityHandler001}
   &EmitterState()
   &VC_FilteredSocketCount()
   &VC_RateQualityHandler()
   &VC_SpawnRibbon_v1(Number of Ribbons=$Emitter.PortCount, Spawn Time=10, SpawnRate=$Emitter.Spawn Rate)
  ]
  particleSpawnScript [
   &VC_PaintedVariantsOverride() {enabled=false}
   &VC_InitializeRibbon_v1(
    Color=$User.PaintedColor
    Custom Offset=true
    Enable Skeletal Position=true
    Flip Facing=true
    Lifetime=1.3
    Local Vector Offset=&AddVectorToPosition(Position=<100 0 0>, Vector=$User.DTCarOffset)
    Offset Amount=1
   )
   &AddVelocity(Velocity=&MakeVector(
     X=-1400
     Y=&Add_Float(
      A=&Sine(
       Normalized Angle=$Emitter.Age
       Period=0.1
       Scale=&FloatFromCurve(
        FloatCurve=DataInterface() {propertyValues {Curve {keys [{tangentMode=RCTM_User, time=0, value=0}, <at 0.6 1>, {tangentMode=RCTM_User, time=1, value=1}]}}}
        Scale Curve=50
       )
       Bias=0
      )
      B=&Sine(
       Normalized Angle=$Emitter.Age
       Period=0.3
       Scale=&FloatFromCurve(
        FloatCurve=DataInterface() {propertyValues {Curve {keys [{tangentMode=RCTM_User, time=0, value=0.1}, <at 0.6 1>, {tangentMode=RCTM_User, time=1, value=1}]}}}
        Scale Curve=50
       )
       Bias=0
      )
     )
     Z=&Sine(
      Normalized Angle=$Emitter.Age
      Period=0.1
      Scale=&FloatFromCurve(
       FloatCurve=DataInterface() {propertyValues {Curve {keys [{tangentMode=RCTM_User, time=0, value=0.1}, <at 0.6 1>, {tangentMode=RCTM_User, time=1, value=1}]}}}
       Scale Curve=55
      )
      Bias=0
     )
    )) {module=AddVelocity001}
   /VehicleCosmetics/FX/Niagara/Modules/MergeEmitters/VC_ExtraBindingColors(ExtraColors=2, R1 Color:LinearColor=$User.PaintedColor_2, R2 Color:LinearColor=$User.PaintedColor_3)
  ]
  particleUpdateScript [
   &ParticleState()
   &InheritSourceMovement(
    Applied Position Scale=&Multiply_VectorByFloat(
     Vector=<1 1 1>
     Float=&FloatFromCurve(FloatCurve=DataInterface() {propertyValues {Curve {keys [
          {interpMode=RCIM_Cubic, leaveTangent=-0.0051, tangentMode=RCTM_Break, time=0, value=1}
          {arriveTangent=-0.247, interpMode=RCIM_Cubic, tangentMode=RCTM_Break, time=1, value=0.9}
         ]}}})
    )
    ApplyForce=false
    ApplyPosition=true
    ApplyPositionFalloff=false
   )
   &SolveForcesAndVelocity(Clamp Velocity=false, Limit Acceleration=false)
   &VC_UpdateRibbon_v1(Enable Skeletal Position=true)
   &ScaleRibbonWidth(Ribbon Width Scale=&FloatFromCurve(
     FloatCurve=DataInterface() {propertyValues {Curve {keys [
         {interpMode=RCIM_Cubic, leaveTangent=1.6099, tangentMode=RCTM_Break, time=0, value=0.85}
         {interpMode=RCIM_Cubic, leaveTangent=-1.5386, tangentMode=RCTM_Break, time=0.08, value=1}
         {arriveTangent=-0.5728, interpMode=RCIM_Cubic, tangentMode=RCTM_Break, time=0.95, value=0}
        ]}}}
     Scale Curve=10
    ))
   &ScaleColor(
    Scale Alpha=&FloatFromCurve(
     FloatCurve=DataInterface() {propertyValues {Curve {keys [
         {arriveTangent=-7.9979, interpMode=RCIM_Cubic, leaveTangent=-7.9979, tangentMode=RCTM_User, time=0, value=0}
         {arriveTangent=12.4362, interpMode=RCIM_Cubic, leaveTangent=26.1544, tangentMode=RCTM_Break, time=0.1, value=-1.25}
         {arriveTangent=9.6937, interpMode=RCIM_Cubic, leaveTangent=4.0462, tangentMode=RCTM_Break, time=0.3, value=3}
         {arriveTangent=-2.3968, interpMode=RCIM_Cubic, leaveTangent=-20.7313, tangentMode=RCTM_Break, time=0.65, value=3}
         <at 1 0 RCIM_Cubic>
        ]}}}
     Scale Curve=2
    )
    Scale RGB=&VectorFromCurve(
     VectorCurve=DataInterface() {propertyValues {XCurve {keys [<at 0 1>,<at 1 1>]}, YCurve {keys [<at 0 1>,<at 1 1>]}, ZCurve {keys [<at 0 1>,<at 1 1>]}}}
     Scale Curve=<6 6 6>
    )
    ScaleA=true
    ScaleRGB=true
   ) {module=ScaleColor002}
   /SetVariables(
    Particles.R1 Color:LinearColor=&self:VC_Dynamic_ScaleColor_03(
     Color to Scale:LinearColor=$Particles.Initial.R1 Color
     RGB Scale:Vector3f=&VectorFromFloat(Value=&FloatFromCurve(FloatCurve=DataInterface() {propertyValues {Curve {keys [<at 0 1>,<at 1 1>]}}}, Scale Curve=6))
     Alpha Scale:NiagaraFloat=&FloatFromCurve(
      FloatCurve=DataInterface() {propertyValues {Curve {keys [
          {arriveTangent=-7.9979, interpMode=RCIM_Cubic, leaveTangent=-7.9979, tangentMode=RCTM_User, time=0, value=0}
          {arriveTangent=12.4362, interpMode=RCIM_Cubic, leaveTangent=26.1544, tangentMode=RCTM_Break, time=0.1, value=-1.25}
          {arriveTangent=9.6937, interpMode=RCIM_Cubic, leaveTangent=4.0462, tangentMode=RCTM_Break, time=0.3, value=3}
          {arriveTangent=-2.3968, interpMode=RCIM_Cubic, leaveTangent=-20.7313, tangentMode=RCTM_Break, time=0.65, value=3}
          <at 1 0 RCIM_Cubic>
         ]}}}
      Scale Curve=2
     )
    )
    Particles.R2 Color:LinearColor=&self:VC_Dynamic_ScaleColor_02(
     Color to Scale:LinearColor=$Particles.Initial.R2 Color
     RGB Scale:Vector3f=<6 6 6>
     Alpha Scale:NiagaraFloat=&FloatFromCurve(
      FloatCurve=DataInterface() {propertyValues {Curve {keys [
          {arriveTangent=-7.9979, interpMode=RCIM_Cubic, leaveTangent=-7.9979, tangentMode=RCTM_User, time=0, value=0}
          {arriveTangent=12.4362, interpMode=RCIM_Cubic, leaveTangent=26.1544, tangentMode=RCTM_Break, time=0.1, value=-1.25}
          {arriveTangent=9.6937, interpMode=RCIM_Cubic, leaveTangent=4.0462, tangentMode=RCTM_Break, time=0.3, value=3}
          {arriveTangent=-2.3968, interpMode=RCIM_Cubic, leaveTangent=-20.7313, tangentMode=RCTM_Break, time=0.65, value=3}
          <at 1 0 RCIM_Cubic>
         ]}}}
      Scale Curve=0.4
     )
    )
    Particles.RibbonWidth2:NiagaraFloat=&FloatFromCurve(FloatCurve=DataInterface() {propertyValues {Curve {keys [<at 0 1>,<at 1 0>]}}}, Scale Curve=20)
    Particles.RibbonWidth3:NiagaraFloat=&FloatFromCurve(
     FloatCurve=DataInterface() {propertyValues {Curve {keys [
         {interpMode=RCIM_Cubic, leaveTangent=-1.1418, tangentMode=RCTM_Break, time=0, value=1}
         {arriveTangent=-1.0164, tangentMode=RCTM_User, time=0.5, value=0.4}
         <at 0.85 0>
        ]}}}
     Scale Curve=122
    )
   ) {isSetParameters=true}
  ]
  rendererClasses [&NiagaraRibbonRendererProperties]
  renderers [
   {rendererClass=&NiagaraRibbonRendererProperties}
   {rendererClass=&NiagaraRibbonRendererProperties, rendererIndex=1}
   {rendererClass=&NiagaraRibbonRendererProperties, rendererIndex=2}
  ]
 }
}
summary {
 systemName=NS_Boost_Firelines_Optimized
 emitters [
  {emitterName=Particles_Flying, rendererClasses [&NiagaraSpriteRendererProperties]}
  {emitterName=Ribbons, rendererClasses [&NiagaraRibbonRendererProperties]}
 ]
 userVariables [
  User.bIsFrontendPreview {type=NiagaraBool}
  User.bIsPainted {type=NiagaraBool}
  User.ColorVariantID {type=NiagaraInt32}
  User.DTCarOffset {type=Vector3f, description="", defaultValue {struct=Vector3f, value=<0 0 0>}}
  User.MainColor {type=LinearColor, description="", defaultValue {struct=LinearColor, value=<rgba 0.0095 0.0354 0.2604 1>}}
  User.PaintedColor {type=LinearColor, description="", defaultValue {struct=LinearColor, value=<rgba 0.102 0.502 1 1.5>}}
  User.PaintedColor_2 {type=LinearColor, description="", defaultValue {struct=LinearColor, value=<rgba 0.102 0.502 1 1.5>}}
  User.PaintedColor_3 {type=LinearColor, description="", defaultValue {struct=LinearColor, value=<rgba 0.7728 0.0827 1 2>}}
  User.PaintedColor_4 {type=LinearColor, description="", defaultValue {struct=LinearColor, value=<rgba 0.8937 0.1546 1 1>}}
  User.PaintedScalar {type=NiagaraFloat, defaultValue=1, description=""}
  User.PortCount {type=NiagaraInt32, defaultValue=2}
 ]
}
