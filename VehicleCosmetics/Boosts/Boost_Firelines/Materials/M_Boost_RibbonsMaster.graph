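# Translucent unlit master material for Niagara ribbon vehicle boost trails, with static switches for variants.
# A panned, rotated or flipbook trail texture has channels masked or blended, then is optionally modulated by a panning mask.
# Particle color, remapped by brightness thresholds, lerps toward a hue-shifted version via Color_Shift.
# Emissive multiplies that color by the pattern, a U-axis fade, dynamic Brightness and optional aurora noise, while particle alpha drives opacity.
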Material /VehicleCosmetics/Boosts/Boost_Firelines/Materials/M_Boost_RibbonsMaster
graph {
 properties {bUsedWithNiagaraRibbons=true, blendMode=BLEND_Translucent, pixelDepthOffsetMode=PDOM_Legacy, shadingModel=MSM_Unlit, usedShadingModels=Unlit}
 nodes [
  n1 = TextureCoordinate() {vTiling=0}
  n2 = ScalarParameter() {defaultValue=0.1, group=Pan, parameterName=PanTile}
  n3 = AppendVector(A=ComponentMask(None=Multiply(A=@n1, B=@n2)) {r=true}, B=ComponentMask(None=TextureCoordinate() {uTiling=0}) {g=true})
  n4 = Multiply(A=Time(), B=ScalarParameter() {defaultValue=-0.5, group=Pan, parameterName=PanSpeed})
  n5 = Panner(Coordinate=@n3, Time=@n4) {speedY=0.4}
  n6 = Constant() {r=0.5}
  n7 = TextureSampleParameter2D(UVs=StaticSwitchParameter(True=Rotator(Coordinate=@n5, Time=@n6) {speed=3.14}, False=Panner(Coordinate=@n3, Time=@n4) {speedX=0.4}) {group=FlipBook, parameterName=RotateFlipbook}) {group=Pan, parameterName=PanMask, texture=/VehicleCosmetics/Boosts/Boost_Firelines/Textures/T_VCVFX_Beam_GlowMask}
  n8 = Panner(Coordinate=TextureCoordinate() {coordinateIndex=1}) {speedX=-2}
  n9 = ScalarParameter() {defaultValue=4, group=FlipBook, parameterName=Columns}
  n10 = ScalarParameter() {defaultValue=4, group=FlipBook, parameterName=Rows}
  n11 = AppendVector(A=@n9, B=@n10)
  n12 = Divide(A=@n8, B=@n11)
  n13 = AppendVector(A=Multiply(A=@n9, B=@n10), B=@n10)
  n14 = Frac(None=Multiply(A=Time(), B=ScalarParameter() {defaultValue=1, group=FlipBook, parameterName=Speed}))
  n15 = AppendVector(A=@n14, B=@n14)
  n16 = Divide(A=Floor(None=Multiply(A=@n13, B=@n15)), B=@n11)
  n17 = Rotator(Coordinate=TextureCoordinate() {coordinateIndex=1, uTiling=2}, Time=Constant() {r=0.5}) {speed=3.14}
  n18 = TextureSampleParameter2D(UVs=StaticSwitchParameter(True=Add(A=@n12, B=@n16), False=StaticSwitchParameter(True=@n17, False=@n8) {group=FlipBook, parameterName=RotateFlipbook}) {group=FlipBook, parameterName=FlipBook}) {group=FlipBook, parameterName=BaseTexture, texture=/VehicleCosmetics/Boosts/Boost_Firelines/Textures/T_VCVFX_TechTrails_RGB}
  n19 = StaticComponentMaskParameter(None=@n18.RGB) {defaultR=true, group=Blend, parameterName=Blend_01}
  n20 = StaticComponentMaskParameter(None=@n18.RGB) {defaultG=true, group=Blend, parameterName=Blend_02}
  n21 = DynamicParameter() {defaultValue=<rgba 1 1 1 0>, paramNames [Brightness, LerpBlend, FlipTexture, ""]}
  n22 = Multiply(A=@n19, B=@n20)
  n23 = Add(A=@n19, B=@n20)
  n24 = StaticSwitchParameter(
   True=StaticSwitchParameter(
    True=LinearInterpolate(A=@n19, B=@n20, Alpha=@n21.LerpBlend)
    False=StaticSwitchParameter(True=@n22, False=@n23) {group=Blend, parameterName="BlendType(T=M)(F=A)"}
   ) {group=Blend, parameterName=LerpBlend}
   False=StaticComponentMaskParameter(None=@n18.RGB) {defaultG=true, group=Blend, parameterName=RGB_Mask}
  ) {defaultValue=true, group=Blend, parameterName=Blend}
  n25 = Multiply(
   A=Multiply(
    A=ScalarParameter() {defaultValue=1, group=Pan, parameterName=PanBrightness}
    B=StaticComponentMaskParameter(None=@n7.RGB) {defaultB=true, group=Pan, parameterName=PanRGB}
   )
   B=@n24
  )
  n26 = StaticSwitchParameter(True=StaticSwitchParameter(True=Add(A=@n25, B=@n24), False=@n25) {defaultValue=true, group=Pan, parameterName=AddPan?}, False=@n24) {defaultValue=true, group=Pan, parameterName=PanLightning}
  MaterialExpressionVertexColor_0 = ParticleColor()
  n27 = Multiply(A=@MaterialExpressionVertexColor_0.A, B=@n26)
  DepthFade(Opacity=@n27) {fadeDistanceDefault=255}
  MaterialExpressionVertexColor_2 = ParticleColor()
  n28 = Constant() {r=0.05}
  n29 = Constant() {r=0.15}
  n30 = Multiply(A=@MaterialExpressionVertexColor_2.RGB, B=Constant())
  n31 = Add(A=@MaterialExpressionVertexColor_2.R, B=@MaterialExpressionVertexColor_2.R)
  MaterialExpressionConstantClamp_4 = Saturate(None=Divide(A=Add(A=@n31, B=@MaterialExpressionVertexColor_2.B), B=Constant() {r=3}))
  n32 = If(
   A=@MaterialExpressionConstantClamp_4
   B=Constant() {r=0.785}
   "A > B"=Multiply(A=Multiply(A=@MaterialExpressionVertexColor_2.RGB, B=@MaterialExpressionVertexColor_2.RGB), B=Constant() {r=0.75})
   "A == B"=@MaterialExpressionVertexColor_2.RGB
   "A < B"=If(
    A=@MaterialExpressionConstantClamp_4
    B=Constant() {r=0.35}
    "A > B"=@n30
    "A == B"=Add(A=@MaterialExpressionVertexColor_2.RGB, B=@n28)
    "A < B"=If(A=@MaterialExpressionConstantClamp_4, B=@n29, "A > B"=@n30, "A == B"=@n30, "A < B"=@n30)
   )
  )
  n33 = Constant() {r=1.5}
  MaterialExpressionVertexColor_1 = ParticleColor()
  n34 = MaterialFunctionCall(Hue Shift Percentage=ScalarParameter() {defaultValue=0.35, parameterName=HueShift}, Texture=@MaterialExpressionVertexColor_1.RGB) {materialFunction=&HueShift}
  n35 = Constant() {r=0.35}
  n36 = Multiply(A=Desaturation(None=@n34.Result, Fraction=Constant() {r=0.15}), B=Constant())
  n37 = Desaturation(None=@n34.Result, Fraction=Constant() {r=0.25})
  n38 = Add(A=@n37, B=Constant() {r=0.05})
  n39 = Add(A=ComponentMask(None=@n37) {r=true}, B=ComponentMask(None=@n37) {g=true})
  n40 = ComponentMask(None=@n37) {b=true}
  MaterialExpressionConstantClamp_3 = Saturate(None=Divide(A=Add(A=@n39, B=@n40), B=Constant() {r=3}))
  n41 = If(A=@MaterialExpressionConstantClamp_3, B=Constant() {r=0.2}, "A > B"=@n36, "A == B"=@n36, "A < B"=@n36)
  n42 = Multiply(
   A=If(
    A=@MaterialExpressionConstantClamp_3
    B=Constant() {r=0.785}
    "A > B"=Multiply(A=@n34.Result, B=@n33)
    "A == B"=@n37
    "A < B"=If(A=@MaterialExpressionConstantClamp_3, B=@n35, "A > B"=@n36, "A == B"=@n38, "A < B"=@n41)
   )
   B=Constant() {r=2.5}
  )
  n43 = ScalarParameter() {parameterName=Color_Shift}
  n44 = Multiply(A=Constant() {r=0.2}, B=@n43)
  MaterialExpressionVertexColor_3 = ParticleColor()
  n45 = Add(A=@MaterialExpressionVertexColor_3.G, B=@MaterialExpressionVertexColor_3.G)
  MaterialExpressionConstantClamp_2 = Saturate(None=Add(A=LinearInterpolate(A=@n43, B=@n44, Alpha=@MaterialExpressionVertexColor_3.R), B=LinearInterpolate(A=@n44, B=@n43, Alpha=@n45)))
  n46 = Constant() {r=8}
  n47 = TextureCoordinate() {coordinateIndex=1, uTiling=0.25, vTiling=0.25}
  n48 = Constant() {r=0.16}
  n49 = Multiply(A=@n21.Brightness, B=Constant() {r=1.25})
  n50 = Multiply(A=Constant() {r=0.1}, B=@n21.Brightness)
  n51 = If(
   A=@MaterialExpressionConstantClamp_4
   B=Constant() {r=0.785}
   "A > B"=Multiply(A=@n21.Brightness, B=Constant() {r=0.25})
   "A == B"=@n21.Brightness
   "A < B"=If(
    A=@MaterialExpressionConstantClamp_4
    B=Constant() {r=0.5}
    "A > B"=@n21.Brightness
    "A == B"=@n21.Brightness
    "A < B"=If(A=@MaterialExpressionConstantClamp_4, B=@n48, "A > B"=@n49, "A == B"=@n49, "A < B"=@n50)
   )
  )
  n52 = StaticSwitchParameter(True=Frac(None=TextureCoordinate() {uTiling=4, vTiling=4}), False=TextureCoordinate()) {group=FlipBook, parameterName=FlipBook}
  MaterialExpressionConstantClamp_5 = Saturate(None=OneMinus(None=ComponentMask(None=@n52) {r=true}))
  MaterialExpressionConstantClamp_6 = Saturate(None=Multiply(A=Multiply(A=@n26, B=@n51), B=@MaterialExpressionConstantClamp_5))
  n53 = Multiply(
   A=TextureSample(UVs=Panner(Coordinate=@n47) {speedX=-0.5, speedY=0.5}) {texture=/DealershipSharedLibrary/FX/Textures/Noise/T_VCVFX_aurora_A}.RGB
   B=@MaterialExpressionConstantClamp_6
  )
  n54 = StaticSwitchParameter(True=Add(A=Multiply(A=@n46, B=@n53), B=@MaterialExpressionConstantClamp_6), False=@MaterialExpressionConstantClamp_6) {group=Pan, parameterName=AddNoise}
  n55 = Multiply(A=@MaterialExpressionConstantClamp_5, B=Constant() {r=10})
  MaterialFunctionCall(Emissive=Multiply(A=LinearInterpolate(A=@n32, B=@n42, Alpha=@MaterialExpressionConstantClamp_2), B=Multiply(A=@n55, B=@n54))) {materialFunction=&MF_FX_EmissiveAdjustments}
  MaterialExpressionConstantClamp_0 = Saturate(None=StaticSwitchParameter(True=@n27, False=@n27) {group=Opacity, parameterName=DBA?})
  Multiply(A=@MaterialExpressionConstantClamp_0)
 ]
}
